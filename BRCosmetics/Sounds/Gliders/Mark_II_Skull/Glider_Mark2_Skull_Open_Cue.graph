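# Sound cue for the opening of the Skull variant of the Mark II glider cosmetic.
# A mixer layers the default Mark II open sound after a brief delay, then a boosted Fortnitemares stinger and a theremin together, then one of four random bone shakes at reduced volume.
# Fixed delay nodes stagger the layers.
# Shared glider open/close attenuation and concurrency apply, along with the Fort_Gliders_Land sound class.
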SoundCue /BRCosmetics/Sounds/Gliders/Mark_II_Skull/Glider_Mark2_Skull_Open_Cue
root=Mixer_0
nodes [
 {id=Delay_0, children [WavePlayer_0], properties {delayMax=0.13, delayMin=0.13}}
 {id=Delay_1, children [WavePlayer_1], properties {delayMax=1.5, delayMin=1.5}}
 {id=Delay_2, children [WavePlayer_2], properties {delayMax=1.5, delayMin=1.5}}
 {id=Delay_3, children [Random_1], properties {delayMax=1.75, delayMin=1.75}}
 {id=Mixer_0, children [Delay_1,Delay_0,Delay_2,Delay_3], properties {inputVolume [2,1,1,0.5]}}
 {id=Random_1, children [WavePlayer_11, WavePlayer_12, WavePlayer_13, WavePlayer_14], properties {weights [1,1,1,1]}}
 {id=WavePlayer_0, properties {soundWaveAssetPtr=/BRCosmetics/Sounds/Gliders/DefaultMarkII/Glider_Mark2_Open}}
 {
  id=WavePlayer_1
  properties {soundWaveAssetPtr=&/FNS14_Fortnitemares_Stinger_-_Mini_Challenge_v1_0_2MX_100bpm}
 }
 {id=WavePlayer_11, properties {soundWaveAssetPtr=&/Skull_Bones_Shake_01}}
 {id=WavePlayer_12, properties {soundWaveAssetPtr=&/Skull_Bones_Shake_02}}
 {id=WavePlayer_13, properties {soundWaveAssetPtr=&/Skull_Bones_Shake_03}}
 {id=WavePlayer_14, properties {soundWaveAssetPtr=&/Skull_Bones_Shake_04}}
 {id=WavePlayer_2, properties {soundWaveAssetPtr=&/Glider_Skull_Mark_II_Therimin}}
]
properties {
 attenuationSettings=/Game/Sounds/Attenuation/Fort_Player_Glider_OpenClose_Athena_Att
 audioPropertiesSheet=/BRCosmetics/Sounds/Settings/PropertySheets/Glider/APS_Cosmetic_Glider_Open
 maxDistance=10000
 soundClassObject=&Fort_Gliders_Land
 volumeMultiplier=0.9
 concurrencySet [&Athena_GliderOpenClose_Conc]
}
